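# Conversation database for the Season 33 Winterfest Case 0 quest, where Tactical Santa recruits the player to investigate naughty-or-nice candidates and sends them to Noir.
# The opening main-questline choice requires an NPC with the TacticalSanta entry tag and the quest to be active.
# A single entry point leads through a linear chain of Speech nodes with one player reply each, ending on a Continue node whose choice carries the entry tag.
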ConversationDatabase /BattlePassS33_Quests/Conversations/Winterfest/Case_0/Quest_S33_Winterfest_P0_Q01
entry_points [{tag=Conversation.Entry.Season33.WF.Quest.00.part.01, destinations [#1]}]
nodes [
 ConversationEntryPointNode_0 {
  id=#1
  class=ConversationEntryPointNode
  outputs [#2]
  properties {entryTag {tagName=Conversation.Entry.Season33.WF.Quest.00.part.01}}
 }
 Speech_C_11 {
  id=#2
  class=Speech_C
  outputs [#3]
  properties {
   softTaskIcon=&T_UI_NPC_Intel2
   generalConfig {defaultMessage="You there! I need help investigating the last few naughty or nice candidates"}
  }
  sub_nodes [
   ConversationChoiceNode_4 {
    class=ConversationChoiceNode
    properties {defaultChoiceDisplayText=Wintervestigation, choiceTags {gameplayTags [{tagName=Conversation.Choice.Category.MainQuestline}]}}
   }
   FortConversationRequirement_ControllerRequirement_1 {
    class=FortConversationRequirement_ControllerRequirement
    properties {
     nodeName=Require: NPC Has Conversation Tag
     participantId {tagName=Conversation.Participant.NPC}
     requirement {class=FortControllerRequirement_NPCHasConversationEntryTag, properties {queryTag {tagName=Conversation.Entry.TacticalSanta}}}
    }
   }
   FortConversationRequirement_ControllerRequirement_2 {
    class=FortConversationRequirement_ControllerRequirement
    properties {
     nodeName=Require: Is Quest Active
     requirement {
      class=FortControllerRequirement_IsQuestActive
      properties {questDefinition=/BattlePassS33_Quests/Items/QuestItems/Winterfest/Case_0/Quest_S33_Winterfest_P0_Q01}
     }
    }
   }
  ]
 }
 Speech_C_28 {
  id=#3
  class=Speech_C
  outputs [#4]
  properties {
   softTaskIcon=&T_UI_NPC_Intel2
   generalConfig {defaultMessage="These candidates are highly likely to be naughty and therefore dangerous, so we have to send in professionals like you"}
  }
  sub_nodes [ConversationChoiceNode_9 {class=ConversationChoiceNode, properties {defaultChoiceDisplayText=Don't you have Elves to help with this sort of thing?}}]
 }
 Speech_C_0 {
  id=#4
  class=Speech_C
  outputs [#5]
  properties {
   softTaskIcon=&T_UI_NPC_Intel2
   generalConfig {defaultMessage="Go see my man on the ground, Noir. He will get you sorted and on track with all of the cases."}
  }
  sub_nodes [ConversationChoiceNode_6 {class=ConversationChoiceNode, properties {defaultChoiceDisplayText=Dangerous??}}]
 }
 Speech_C_4 {
  id=#5
  class=Speech_C
  outputs [#6]
  properties {
   softTaskIcon=&T_UI_NPC_Intel2
   generalConfig {defaultMessage="And remember, this holiday season relies entirely on YOU."}
  }
  sub_nodes [ConversationChoiceNode_7 {class=ConversationChoiceNode, properties {defaultChoiceDisplayText=...}}]
 }
 Continue_C_1 {
  id=#6
  class=Continue_C
  properties {nodeName="", softTaskIcon=&T_UI_NPC_Intel2}
  sub_nodes [ConversationChoiceNode_0 {
    class=ConversationChoiceNode
    properties {defaultChoiceDisplayText="No pressure, then...", choiceTags {gameplayTags [{tagName=Conversation.Entry.Season33.WF.Quest.00.part.01}]}}
   }]
 }
]
