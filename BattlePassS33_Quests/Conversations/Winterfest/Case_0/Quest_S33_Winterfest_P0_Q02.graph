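# Conversation database for Winterfest Case 0, part 02 "Wintervestigation" quest dialogue, introducing the player as a rookie detective.
# One entry point offers two branches, each requiring the active quest and a specific NPC conversation tag.
# GumshoeDark plays a chained briefing about upcoming cases and helping Guff, ending in a Continue choice tagged with the entry tag.
# TacticalSanta repeats a reminder to see Noir, then returns to start.
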ConversationDatabase /BattlePassS33_Quests/Conversations/Winterfest/Case_0/Quest_S33_Winterfest_P0_Q02
entry_points [{tag=Conversation.Entry.Season33.WF.Quest.00.part.02, destinations [#1]}]
nodes [
 ConversationEntryPointNode_0 {
  id=#1
  class=ConversationEntryPointNode
  outputs [#2, #3]
  properties {entryTag {tagName=Conversation.Entry.Season33.WF.Quest.00.part.02}}
 }
 Speech_C_1 {
  id=#3
  class=Speech_C
  outputs [#4]
  properties {
   softTaskIcon=&T_UI_NPC_Intel2
   generalConfig {defaultMessage=I keep my office small so that people can tell I'm smart.}
  }
  sub_nodes [
   FortConversationRequirement_ControllerRequirement_12 {
    class=FortConversationRequirement_ControllerRequirement
    properties {
     nodeName=Require: Is Quest Active
     requirement {
      class=FortControllerRequirement_IsQuestActive
      properties {questDefinition=/BattlePassS33_Quests/Items/QuestItems/Winterfest/Case_0/Quest_S33_Winterfest_P0_Q02}
     }
    }
   }
   ConversationChoiceNode_15 {
    class=ConversationChoiceNode
    properties {defaultChoiceDisplayText=Wintervestigation, choiceTags {gameplayTags [{tagName=Conversation.Choice.Category.MainQuestline}]}}
   }
   FortConversationRequirement_ControllerRequirement_13 {
    class=FortConversationRequirement_ControllerRequirement
    properties {
     nodeName=Require: NPC Has Conversation Tag
     participantId {tagName=Conversation.Participant.NPC}
     requirement {class=FortControllerRequirement_NPCHasConversationEntryTag, properties {queryTag {tagName=Conversation.Entry.GumshoeDark}}}
    }
   }
  ]
 }
 Speech_C_7 {
  id=#4
  class=Speech_C
  outputs [#5]
  properties {
   softTaskIcon=&T_UI_NPC_Intel2
   generalConfig {defaultMessage=You must be the new rookie. Sgt. Winters let me know you'd be stopping by.}
  }
  sub_nodes [ConversationChoiceNode_16 {class=ConversationChoiceNode, properties {defaultChoiceDisplayText=It doesn't seem so small to me...}}]
 }
 Speech_C_10 {
  id=#5
  class=Speech_C
  outputs [#6]
  properties {
   softTaskIcon=&T_UI_NPC_Intel2
   generalConfig {defaultMessage=We'll have new cases coming in every few days as he narrows down the Naughty List.}
  }
  sub_nodes [ConversationChoiceNode_1 {class=ConversationChoiceNode, properties {defaultChoiceDisplayText=That's me.}}]
 }
 Speech_C_0 {
  id=#6
  class=Speech_C
  outputs [#7]
  properties {
   softTaskIcon=&T_UI_NPC_Intel2
   generalConfig {defaultMessage="For now, help Junior Detective Guff with his daily assignments and get outta my hair."}
  }
  sub_nodes [ConversationChoiceNode_17 {class=ConversationChoiceNode, properties {defaultChoiceDisplayText=...}}]
 }
 Continue_C_0 {
  id=#7
  class=Continue_C
  properties {softTaskIcon=&T_UI_NPC_Intel2}
  sub_nodes [ConversationChoiceNode_0 {
    class=ConversationChoiceNode
    properties {defaultChoiceDisplayText=Alright then..., choiceTags {gameplayTags [{tagName=Conversation.Entry.Season33.WF.Quest.00.part.02}]}}
   }]
 }
 Speech_C_6 {
  id=#2
  class=Speech_C
  outputs [#8]
  properties {
   softTaskIcon=&T_UI_NPC_Intel2
   generalConfig {defaultMessage="I Already told you! Go see my man on the ground, Noir. He will get you sorted and on track with all of the cases. And don't forget, this holiday season depends almost entirely on you alone!"}
  }
  sub_nodes [
   ConversationChoiceNode_8 {
    class=ConversationChoiceNode
    properties {defaultChoiceDisplayText=Wintervestigation, choiceTags {gameplayTags [{tagName=Conversation.Choice.Category.MainQuestline}]}}
   }
   FortConversationRequirement_ControllerRequirement_10 {
    class=FortConversationRequirement_ControllerRequirement
    properties {
     nodeName=Require: Is Quest Active
     requirement {
      class=FortControllerRequirement_IsQuestActive
      properties {questDefinition=&Quest_S33_Winterfest_P0_Q02}
     }
    }
   }
   FortConversationRequirement_ControllerRequirement_11 {
    class=FortConversationRequirement_ControllerRequirement
    properties {
     nodeName=Require: NPC Has Conversation Tag
     participantId {tagName=Conversation.Participant.NPC}
     requirement {class=FortControllerRequirement_NPCHasConversationEntryTag, properties {queryTag {tagName=Conversation.Entry.TacticalSanta}}}
    }
   }
  ]
 }
 ReturnToStart_C_0 {id=#8, class=ReturnToStart_C, properties {softTaskIcon=&T_UI_NPC_MainQuest}}
]
