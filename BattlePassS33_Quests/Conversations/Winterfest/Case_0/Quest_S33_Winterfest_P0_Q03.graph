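# Conversation database for the Season 33 Winterfest Case 0 Q03 quest, giving NPC dialogue under the "Wintervestigation" main-questline choice.
# Its entry point branches to two speech options, both requiring the GumshoeDark NPC tag, active Q03 and no active Case 1–6 quests.
# An active daily quest selects a hint to help Guff; otherwise the NPC reports no new cases.
# Both lines return to the conversation start.
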ConversationDatabase /BattlePassS33_Quests/Conversations/Winterfest/Case_0/Quest_S33_Winterfest_P0_Q03
entry_points [{tag=Conversation.Entry.Season33.WF.Quest.00.part.03, destinations [#1]}]
nodes [
 ConversationEntryPointNode_0 {
  id=#1
  class=ConversationEntryPointNode
  outputs [#2, #3]
  properties {entryTag {tagName=Conversation.Entry.Season33.WF.Quest.00.part.03}}
 }
 Speech_C_2 {
  id=#3
  class=Speech_C
  outputs [#4]
  properties {
   softTaskIcon=&T_UI_NPC_Intel2
   generalConfig {defaultMessage="Don't have much for ya, but I think Guff might need some help with his daily assignment."}
  }
  sub_nodes [
   FortConversationRequirement_ControllerRequirement_10 {
    class=FortConversationRequirement_ControllerRequirement
    properties {
     nodeName=Require: NPC Has Conversation Tag
     participantId {tagName=Conversation.Participant.NPC}
     requirement {class=FortControllerRequirement_NPCHasConversationEntryTag, properties {queryTag {tagName=Conversation.Entry.GumshoeDark}}}
    }
   }
   FortConversationRequirement_ControllerRequirement_11 {
    class=FortConversationRequirement_ControllerRequirement
    properties {
     nodeName=Require: Is Quest Active
     requirement {
      class=FortControllerRequirement_IsQuestActive
      properties {questDefinition=/BattlePassS33_Quests/Items/QuestItems/Winterfest/Case_0/Quest_S33_Winterfest_P0_Q03}
     }
    }
   }
   FortConversationRequirement_ControllerRequirement_12 {
    class=FortConversationRequirement_ControllerRequirement
    properties {
     nodeName=Require: Is Quest Active
     requirement {
      class=FortControllerRequirement_IsQuestActive
      properties {
       questDefinition=/BattlePassS33_Quests/Items/QuestItems/Winterfest/Case_0/Quest_S33_Winterfest_P0_Q02
       testPolicy=AnyOf
       questDefinitions [
        /BattlePassS33_Quests/Items/QuestItems/Winterfest/Daily_Case/Quest_S33_Winterfest_D_Q01
        /BattlePassS33_Quests/Items/QuestItems/Winterfest/Daily_Case/Quest_S33_Winterfest_D_Q02
        /BattlePassS33_Quests/Items/QuestItems/Winterfest/Daily_Case/Quest_S33_Winterfest_D_Q03
        /BattlePassS33_Quests/Items/QuestItems/Winterfest/Daily_Case/Quest_S33_Winterfest_D_Q04
        /BattlePassS33_Quests/Items/QuestItems/Winterfest/Daily_Case/Quest_S33_Winterfest_D_Q05
        /BattlePassS33_Quests/Items/QuestItems/Winterfest/Daily_Case/Quest_S33_Winterfest_D_Q06
        /BattlePassS33_Quests/Items/QuestItems/Winterfest/Daily_Case/Quest_S33_Winterfest_D_Q07
        /BattlePassS33_Quests/Items/QuestItems/Winterfest/Daily_Case/Quest_S33_Winterfest_D_Q08
        /BattlePassS33_Quests/Items/QuestItems/Winterfest/Daily_Case/Quest_S33_Winterfest_D_Q09
        /BattlePassS33_Quests/Items/QuestItems/Winterfest/Daily_Case/Quest_S33_Winterfest_D_Q10
       ]
      }
     }
    }
   }
   ConversationChoiceNode_1 {
    class=ConversationChoiceNode
    properties {
     defaultChoiceDisplayText=Wintervestigation: Guff's Daily Assignment
     choiceTags {gameplayTags [{tagName=Conversation.Choice.Category.MainQuestline}]}
    }
   }
   FortConversationRequirement_ControllerRequirement_13 {
    class=FortConversationRequirement_ControllerRequirement
    properties {
     nodeName=Require: Is Quest Active
     requirement {
      class=FortControllerRequirement_IsQuestActive
      properties {
       bInverseCondition=true
       questDefinition=&Quest_S33_Winterfest_P0_Q02
       testPolicy=AnyOf
       questDefinitions [
        /BattlePassS33_Quests/Items/QuestItems/Winterfest/Case_1/Quest_S33_Winterfest_P1_Q01
        /BattlePassS33_Quests/Items/QuestItems/Winterfest/Case_1/Quest_S33_Winterfest_P1_Q02
        /BattlePassS33_Quests/Items/QuestItems/Winterfest/Case_1/Quest_S33_Winterfest_P1_Q03
        /BattlePassS33_Quests/Items/QuestItems/Winterfest/Case_1/Quest_S33_Winterfest_P1_Q04
        /BattlePassS33_Quests/Items/QuestItems/Winterfest/Case_1/Quest_S33_Winterfest_P1_Q05
        /BattlePassS33_Quests/Items/QuestItems/Winterfest/Case_2/Quest_S33_Winterfest_P2_Q01
        /BattlePassS33_Quests/Items/QuestItems/Winterfest/Case_2/Quest_S33_Winterfest_P2_Q02
        /BattlePassS33_Quests/Items/QuestItems/Winterfest/Case_2/Quest_S33_Winterfest_P2_Q03
        /BattlePassS33_Quests/Items/QuestItems/Winterfest/Case_2/Quest_S33_Winterfest_P2_Q04
        /BattlePassS33_Quests/Items/QuestItems/Winterfest/Case_2/Quest_S33_Winterfest_P2_Q05
        /BattlePassS33_Quests/Items/QuestItems/Winterfest/Case_3/Quest_S33_Winterfest_P3_Q01
        /BattlePassS33_Quests/Items/QuestItems/Winterfest/Case_3/Quest_S33_Winterfest_P3_Q02
        /BattlePassS33_Quests/Items/QuestItems/Winterfest/Case_3/Quest_S33_Winterfest_P3_Q03
        /BattlePassS33_Quests/Items/QuestItems/Winterfest/Case_3/Quest_S33_Winterfest_P3_Q04
        /BattlePassS33_Quests/Items/QuestItems/Winterfest/Case_3/Quest_S33_Winterfest_P3_Q05
        /BattlePassS33_Quests/Items/QuestItems/Winterfest/Case_4/Quest_S33_Winterfest_P4_Q01
        /BattlePassS33_Quests/Items/QuestItems/Winterfest/Case_4/Quest_S33_Winterfest_P4_Q02
        /BattlePassS33_Quests/Items/QuestItems/Winterfest/Case_4/Quest_S33_Winterfest_P4_Q03
        /BattlePassS33_Quests/Items/QuestItems/Winterfest/Case_4/Quest_S33_Winterfest_P4_Q04
        /BattlePassS33_Quests/Items/QuestItems/Winterfest/Case_4/Quest_S33_Winterfest_P4_Q05
        /BattlePassS33_Quests/Items/QuestItems/Winterfest/Case_5/Quest_S33_Winterfest_P5_Q01
        /BattlePassS33_Quests/Items/QuestItems/Winterfest/Case_5/Quest_S33_Winterfest_P5_Q02
        /BattlePassS33_Quests/Items/QuestItems/Winterfest/Case_5/Quest_S33_Winterfest_P5_Q03
        /BattlePassS33_Quests/Items/QuestItems/Winterfest/Case_5/Quest_S33_Winterfest_P5_Q04
        /BattlePassS33_Quests/Items/QuestItems/Winterfest/Case_5/Quest_S33_Winterfest_P5_Q05
        /BattlePassS33_Quests/Items/QuestItems/Winterfest/Case_6/Quest_S33_Winterfest_P6_Q01
        /BattlePassS33_Quests/Items/QuestItems/Winterfest/Case_6/Quest_S33_Winterfest_P6_Q02
        /BattlePassS33_Quests/Items/QuestItems/Winterfest/Case_6/Quest_S33_Winterfest_P6_Q03
        /BattlePassS33_Quests/Items/QuestItems/Winterfest/Case_6/Quest_S33_Winterfest_P6_Q04
        /BattlePassS33_Quests/Items/QuestItems/Winterfest/Case_6/Quest_S33_Winterfest_P6_Q05
       ]
      }
     }
    }
   }
  ]
 }
 ReturnToStart_C_1 {id=#4, class=ReturnToStart_C, properties {softTaskIcon=&T_UI_NPC_Intel2}}
 Speech_C_9 {
  id=#2
  class=Speech_C
  outputs [#4]
  properties {
   softTaskIcon=&T_UI_NPC_Intel2
   generalConfig {defaultMessage=No new cases. Go get some air and check back in later.}
  }
  sub_nodes [
   FortConversationRequirement_ControllerRequirement_19 {
    class=FortConversationRequirement_ControllerRequirement
    properties {
     nodeName=Require: NPC Has Conversation Tag
     participantId {tagName=Conversation.Participant.NPC}
     requirement {class=FortControllerRequirement_NPCHasConversationEntryTag, properties {queryTag {tagName=Conversation.Entry.GumshoeDark}}}
    }
   }
   FortConversationRequirement_ControllerRequirement_20 {
    class=FortConversationRequirement_ControllerRequirement
    properties {
     nodeName=Require: Is Quest Active
     requirement {
      class=FortControllerRequirement_IsQuestActive
      properties {
       bInverseCondition=true
       questDefinition=&Quest_S33_Winterfest_P0_Q02
       testPolicy=AnyOf
       questDefinitions [
        &Quest_S33_Winterfest_P1_Q01
        &Quest_S33_Winterfest_P1_Q02
        &Quest_S33_Winterfest_P1_Q03
        &Quest_S33_Winterfest_P1_Q04
        &Quest_S33_Winterfest_P1_Q05
        &Quest_S33_Winterfest_P2_Q01
        &Quest_S33_Winterfest_P2_Q02
        &Quest_S33_Winterfest_P2_Q03
        &Quest_S33_Winterfest_P2_Q04
        &Quest_S33_Winterfest_P2_Q05
        &Quest_S33_Winterfest_P3_Q01
        &Quest_S33_Winterfest_P3_Q02
        &Quest_S33_Winterfest_P3_Q03
        &Quest_S33_Winterfest_P3_Q04
        &Quest_S33_Winterfest_P3_Q05
        &Quest_S33_Winterfest_P4_Q01
        &Quest_S33_Winterfest_P4_Q02
        &Quest_S33_Winterfest_P4_Q03
        &Quest_S33_Winterfest_P4_Q04
        &Quest_S33_Winterfest_P4_Q05
        &Quest_S33_Winterfest_P5_Q01
        &Quest_S33_Winterfest_P5_Q02
        &Quest_S33_Winterfest_P5_Q03
        &Quest_S33_Winterfest_P5_Q04
        &Quest_S33_Winterfest_P5_Q05
        &Quest_S33_Winterfest_P6_Q01
        &Quest_S33_Winterfest_P6_Q02
        &Quest_S33_Winterfest_P6_Q03
        &Quest_S33_Winterfest_P6_Q04
        &Quest_S33_Winterfest_P6_Q05
        &Quest_S33_Winterfest_D_Q01
        &Quest_S33_Winterfest_D_Q02
        &Quest_S33_Winterfest_D_Q03
        &Quest_S33_Winterfest_D_Q04
        &Quest_S33_Winterfest_D_Q05
        &Quest_S33_Winterfest_D_Q06
        &Quest_S33_Winterfest_D_Q07
        &Quest_S33_Winterfest_D_Q08
        &Quest_S33_Winterfest_D_Q09
        &Quest_S33_Winterfest_D_Q10
       ]
      }
     }
    }
   }
   ConversationChoiceNode_8 {
    class=ConversationChoiceNode
    properties {defaultChoiceDisplayText=Wintervestigation, choiceTags {gameplayTags [{tagName=Conversation.Choice.Category.MainQuestline}]}}
   }
   FortConversationRequirement_ControllerRequirement_23 {
    class=FortConversationRequirement_ControllerRequirement
    properties {
     nodeName=Require: Is Quest Active
     requirement {
      class=FortControllerRequirement_IsQuestActive
      properties {questDefinition=&Quest_S33_Winterfest_P0_Q03}
     }
    }
   }
  ]
 }
]
